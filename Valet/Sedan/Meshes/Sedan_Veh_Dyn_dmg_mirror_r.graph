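# Physics asset for the sedan's damaged right mirror mesh, giving it simple collision bodies and a jointed link.
# Two capsule bodies cover it: a large one on the body bone and a small one on the mirror_r bone.
# A single constraint joins mirror_r to body with swing and twist limit angles set.
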PhysicsAsset /Valet/Sedan/Meshes/Sedan_Veh_Dyn_dmg_mirror_r
bodies {
 body [{shapeType=Capsule, length=77.8438, radius=40, center=<-3.643 0.0005 -0.0002>, rotation=<rot 0 0 89.9997>}]
 mirror_r [{shapeType=Capsule, length=14.6562, radius=10, center=<-0.0001 15.4481 8.4562>, rotation=<rot 0 0 -90>}]
}
constraints [{bone1Name=mirror_r, bone2Name=body, swing1LimitDegrees=10, swing2LimitDegrees=30, twistLimitDegrees=20}]
